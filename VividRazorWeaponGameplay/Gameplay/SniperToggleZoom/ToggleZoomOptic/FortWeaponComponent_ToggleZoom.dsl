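; Weapon component for a sniper optic that toggles between default and zoomed scope levels.
; Target-mode changes set the first-person mesh FOV threshold and drive a reversible timer transition that curve-lerps optic FOV and scope crosshair scale.
; A newly started transition also plays scope in/out audio for locally viewed pawns.
; Targeting applies or removes a first-person camera GameplayEffect, which unequip and EndPlay also remove, while equipping resyncs scope parameters to the current mode.
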(event EventBeginPlay
  (|CallParentFunction)
  ;[0 0.15 0.1392 0.5] Event bindings
  (EventDispatchers|BindEventtoOnTargetModeChanged (EventDispatchers|CreateEvent :!event Custom|TargetModeChanged))
  (bind _weapon (Utilities|GetWeapon)
    (:Is%20Valid
      (EventDispatchers|BindEventtoOnTargetingChanged _weapon (EventDispatchers|CreateEvent :!event Custom|TargetingChanged))
      (bind _fortpawn (Utilities|GetPawn)
        (:Is%20Valid
          (FortPawn|BindEventtoOnWeaponEquippedDelegate _fortpawn (EventDispatchers|CreateEvent :!event Custom|EquippedWeaponChanged)))))))

(event EventEndPlay
  (params (EndPlayReason :type EEndPlayReason%20Enum :default "Destroyed"))
  (|CallParentFunction EndPlayReason)
  (CallFunction|WeaponUnequipped)
  ;[0 0.15 0.1392 0.5] Event bindings
  (EventDispatchers|UnbindEventfromOnTargetModeChanged (EventDispatchers|CreateEvent :!event Custom|TargetModeChanged))
  (bind _weapon (Utilities|GetWeapon)
    (:Is%20Valid
      (EventDispatchers|UnbindEventfromOnTargetingChanged _weapon (EventDispatchers|CreateEvent :!event Custom|TargetingChanged))
      (bind _fortpawn (Utilities|GetPawn)
        (:Is%20Valid
          (FortPawn|UnbindEventfromOnWeaponEquippedDelegate _fortpawn (EventDispatchers|CreateEvent :!event Custom|EquippedWeaponChanged)))))))

(event Custom|TargetModeChanged
  (params (TargetMode :type Integer))
  (Variables|ZoomTransition|SetbTransitioningForward (== TargetMode 1))
  (bind _component (Utilities|GetFirstPersonMeshComponent)
    (:Is%20Valid
      (Variables|/Script/FortniteGame.FortWeaponComp_FirstPersonMesh|SetFirstPersonMeshFOVThreshold (Utilities|Select :Index TargetMode :Option%200 (Variables|Default|GetDefaultFirstPersonMeshThreshold) :Option%201 (Variables|Default|GetZoomedFirstPersonMeshThreshold)) _component))
    (:Is%20Not%20Valid
      _))
  ;[0.15 0.1218 0 0.5] Start zoom transition
  (if (not (Utilities|Time|IsValidTimerHandle (Variables|ZoomTransition|GetTransitionUpdateTimerHandle)))
    (bind _returnvalue (Utilities|Time|SetTimerbyEvent (EventDispatchers|CreateEvent :!event Custom|UpdateZoomTransition) 0.01 true true))
    (Variables|ZoomTransition|SetTransitionUpdateTimerHandle _returnvalue)
    (Variables|ZoomTransition|SetLastUpdateWorldTime (Utilities|Time|GetTimeSeconds))
    (CallFunction|ToggleZoomAudio (Variables|ZoomTransition|GetbTransitioningForward))))

(event Custom|UpdateZoomTransition
  ;[0.15 0.02427 0 0.5] Zoom transition tick
  (bind _alpha (CallFunction|UpdateTransitionAlpha))
  (CallFunction|LerpScopeParams _alpha)
  (Variables|ZoomTransition|SetLastUpdateWorldTime (Utilities|Time|GetTimeSeconds))
  (if (select (Variables|ZoomTransition|GetbTransitioningForward) (>= (Variables|ZoomTransition|GetTransitionAlpha) 1.0) (>= 0 (Variables|ZoomTransition|GetTransitionAlpha)))
    (Utilities|Time|ClearandInvalidateTimerbyHandle (Variables|ZoomTransition|GetTransitionUpdateTimerHandle))))

(event Custom|TargetingChanged
  (params (bIsTargeting :type Boolean))
  (if bIsTargeting
    (bind _fortpawn (Utilities|GetPawn)
      (:Is%20Valid
        (bind _returnvalue (GameplayEffects|ApplyGameplayEffectToSelf (Ability|GetAbilitySystemComponent _fortpawn) "/VividRazorWeaponGameplay/Gameplay/SniperToggleZoom/ToggleZoomOptic/GE_CameraMode_FirstPerson_ToggleZoom.GE_CameraMode_FirstPerson_ToggleZoom_C" 0.0 (Class|AthenaFunctionLibrary|MakeSimpleFortEffectContext _fortpawn)))
        (Variables|Default|SetGameplayEffectHandle _returnvalue)))
    (else
      (bind _fortpawn_2 (Utilities|GetPawn)
        (:Is%20Valid
          (GameplayEffects|RemoveActiveGameplayEffect (Ability|GetAbilitySystemComponent _fortpawn_2) (Variables|Default|GetGameplayEffectHandle) -1))))))

(event Custom|WeaponUnequipped
  (bind _fortpawn (Utilities|GetPawn)
    (:Is%20Valid
      (GameplayEffects|RemoveActiveGameplayEffect (Ability|GetAbilitySystemComponent _fortpawn) (Variables|Default|GetGameplayEffectHandle) -1))))

(event Custom|WeaponEquipped
  (CallFunction|LerpScopeParams (Math|Conversions|ToFloat(Integer) (Variables|FortWeaponComponent_TargetModeSelector|GetTargetMode))))

(event Custom|EquippedWeaponChanged
  (params (NewWeapon :type Fort%20Weapon%20Object%20Reference) (PrevWeapon :type Fort%20Weapon%20Object%20Reference))
  (if (!= NewWeapon PrevWeapon)
    (bind _weapon (Utilities|GetWeapon)
      (:Is%20Valid
        (if (== _weapon NewWeapon)
          (CallFunction|WeaponEquipped)
          (elif (== _weapon PrevWeapon)
            (CallFunction|WeaponUnequipped)))))))

(fn LerpScopeParams
  (params (Alpha :type Float%20%28double-precision%29))
  (bind _transitioncurve (Variables|ZoomTransition|GetTransitionCurve))
  (bind _returnvalue (Math|Curves|GetFloatValue _transitioncurve Alpha))
  (bind _weapon (Utilities|GetWeapon)
    (:Is%20Valid
      (bind _output_get (LocalVariables|SetWeapon _weapon))
      ;[0.0441 0 0.15 0.5] Set crosshair scale
      (bind _returnvalue_2 (Scope|GetorCreateScopePostProcessMaterialInstanceDynamics _output_get))
      (bind (_output_weight _output_object) (Utilities|Struct|BreakWeightedBlendable (Utilities|Array|Get(aref) _returnvalue_2)))
      (bind _asmaterial_instance_dynamic (Utilities|Casting|CastToMaterialInstanceDynamic _output_object))
      (Rendering|Material|SetScalarParameterValue _asmaterial_instance_dynamic (Variables|MaterialParameters|GetCrosshairScaleName) (Math|Float|Lerp (Variables|Zoom|GetDefaultCrosshairScale) (Variables|Zoom|GetZoomedCrosshairScale) _returnvalue) :!class /Script/Engine.MaterialInstanceDynamic)
      ;[0 0.15 0.01323 0.5] Set FOV
      (Utilities|Struct|SetmembersinFortWeaponOpticRuntimeData (Variables|/Script/FortniteGame.FortWeapon|GetOpticRuntimeData (LocalVariables|GetWeapon)) (Math|Float|Lerp (Variables|Zoom|GetDefaulltFOV) (ScalableFloat|GetValueatLevel (Variables|Default|GetZoomed%20FOV%20Hotfix)) _returnvalue)))))

(fn ToggleZoomAudio
  (params (ScopeIn :type Boolean))
  (Networking|CanExecuteCosmeticEvents
    (:True
      (bind _fortpawn (Utilities|GetPawn)
        (:Is%20Valid
          (if (Pawn|IsLocallyViewed _fortpawn)
            (bind _weapon (Utilities|GetWeapon)
              (:Is%20Valid
                (Utilities|IsValid (SoundLibrary|GetSoundLibraryComponent _weapon "(TagName=\"SoundLibrary.Weapon\")" :!class /Script/FortniteGame.FortWeapon)
                  (:Is%20Valid
                    (SoundLibrary|PlayEvent (SoundLibrary|GetSoundLibraryComponent _weapon "(TagName=\"SoundLibrary.Weapon\")" :!class /Script/FortniteGame.FortWeapon) (select (Variables|GetScopeIn) "(TagName=\"SoundLibrary.Weapon.Scope.In\")" "(TagName=\"SoundLibrary.Weapon.Scope.Out\")"))))))))))))

(fn UpdateTransitionAlpha
  (bind _output_get (Variables|ZoomTransition|SetTransitionAlpha (+ (Variables|ZoomTransition|GetTransitionAlpha) (select (Variables|ZoomTransition|GetbTransitioningForward) (/ (- (Utilities|Time|GetTimeSeconds) (Variables|ZoomTransition|GetLastUpdateWorldTime)) (Variables|ZoomTransition|GetTransitionDuration)) (* (/ (- (Utilities|Time|GetTimeSeconds) (Variables|ZoomTransition|GetLastUpdateWorldTime)) (Variables|ZoomTransition|GetTransitionDuration)) -1.0)))))
  (return _output_get))
